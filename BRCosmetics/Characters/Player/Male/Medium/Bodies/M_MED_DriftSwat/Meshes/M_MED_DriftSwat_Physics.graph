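# Physics asset for the DriftSwat medium male player body cosmetic, covering only the bones of its hood.
# It gives a capsule body to the base hood bone and to each of two dynamic hood bones, with a much smaller capsule on the base.
# Two constraints link them in a chain from base to tip, with zero twist limits and far wider swing on one axis than the other, tightening toward the tip.
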PhysicsAsset /BRCosmetics/Characters/Player/Male/Medium/Bodies/M_MED_DriftSwat/Meshes/M_MED_DriftSwat_Physics
bodies {
 base_hood_01_m [{shapeType=Capsule, length=0.505, radius=0.505, shapeName=base_hood_01_m_capsule, rotation=<rot 0 -90 0>}]
 dyn_hood_01_m [{shapeType=Capsule, length=3, radius=2, shapeName=dyn_hood_01_m_capsule, center=<2.4894 0 0>, rotation=<rot 90 0 0>}]
 dyn_hood_02_m [{shapeType=Capsule, length=3, radius=2, shapeName=dyn_hood_02_m_capsule, center=<2.3606 0 0>, rotation=<rot 90 0 0>}]
}
constraints [
 {bone1Name=dyn_hood_02_m, bone2Name=dyn_hood_01_m, swing1LimitDegrees=5, swing2LimitDegrees=50, twistLimitDegrees=0}
 {bone1Name=dyn_hood_01_m, bone2Name=base_hood_01_m, swing1LimitDegrees=10, swing2LimitDegrees=70, twistLimitDegrees=0}
]
