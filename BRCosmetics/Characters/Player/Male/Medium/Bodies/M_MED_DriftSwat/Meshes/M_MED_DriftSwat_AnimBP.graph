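# Animation Blueprint for the DriftSwat male body cosmetic, adding procedural and secondary motion to hood, zipper, knife, pouches, belt and shoe pulls.
# The pose copied from the attached parent runs through a deform Control Rig, weighted constraints and state-driven bone offsets.
# Cached stages then feed RigidBody and AnimDynamics simulation, followed by a look-down zipper tilt.
# The EventGraph sets gravity, alphas and sim-space settings from skydive, jump, crouch and neck state.
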AnimBlueprint /BRCosmetics/Characters/Player/Male/Medium/Bodies/M_MED_DriftSwat/Meshes/M_MED_DriftSwat_AnimBP
target_skeleton=&/M_MED_DriftSwat_Skeleton
graphs [
 AnimGraph {nodes [
   "Constraint - base_knife_l (PPPPPPPP)"!(
    ComponentPose
    Alpha=1
    ConstraintWeights_0=0.3697
    ConstraintWeights_1:float=0.2201
    ConstraintWeights_2:float=0.109
    ConstraintWeights_3:float=0.101
    ConstraintWeights_4:float=0.0652
    ConstraintWeights_5:float=0.053
    ConstraintWeights_6:float=0.0419
    ConstraintWeights_7:float=0.0401
   ) -> (Pose) {
    alphaBoolBlend {blendOption=Linear}
    boneToModify {boneName=base_knife_l}
    constraintSetup [
     {targetBone {boneName=spine_05}}
     {targetBone {boneName=clavicle_l}}
     {targetBone {boneName=spine_04}}
     {targetBone {boneName=deform_pec_l}}
     {targetBone {boneName=deform_upperarm_up_l}}
     {targetBone {boneName=spine_03}}
     {targetBone {boneName=deform_upperarm_fr_l}}
     {targetBone {boneName=upperarm_twist_01_l}}
    ]
    constraintWeights [1,1,1,1,1,1,1,1]
   }
   Output Pose(Result=Use cached pose 'CRCache_Zipper'()) {layerGroup=None}
   n1 = Control Rig(Source=Copy Pose From Mesh() {bUseAttachedParent=true}) {
    alphaBoolBlend {blendOption=Linear}
    controlRigAssetReference {blueprintRigClass=&/M_MED_DeformRig_DriftSwat_ControlRig_new.M_MED_DeformRig_DriftSwat_ControlRig_new_C}
   }
   Save cached pose 'CRCache_ControlRig'(Pose=@n1) {cacheName=CRCache_ControlRig}
   n2 = Get GravityOverride()
   n3 = RigidBody!(
    ComponentPose=Local To Component(LocalPose=Use cached pose 'CRCache_Legs'())
    Alpha=Get AlphaDynHood() -> (AlphaDynHood:double)
    OverrideWorldGravity:Vector=@n2
    ExternalForce
    SimSpaceSettings:Sim Space Settings Structure=Get Sim Space Settings SkyDive() -> (Sim Space Settings SkyDive:Sim Space Settings Structure)
   ) -> (Pose) {
    bOverrideWorldGravity=true
    overridePhysicsAsset=&/M_MED_DriftSwat_Physics
    simulationSpace=BaseBoneSpace
    alphaBoolBlend {blendOption=Linear}
    alphaScaleBiasClamp {bInterpResult=true}
    baseBoneRef {boneName=spine_03}
    componentAppliedLinearAccClamp=<3000 3000 3000>
    componentLinearAccScale=<0.3 0.4 0.3>
    componentLinearVelScale=<0.3 0.4 0.3>
    simSpaceSettings {
     maxAngularAcceleration=1000
     maxAngularVelocity=3
     maxLinearAcceleration=700
     maxLinearVelocity=500
     velocityScaleZ=0
     worldAlpha=0.65
     externalLinearDragV=<0.2 0.2>
    }
   }
   n4 = "Constraint - base_hoodtop_m (R)"(ComponentPose=@n3, Alpha=Get AlphaConstraintHood() -> (AlphaConstraintHood:double), ConstraintWeights_0=1) {
    alphaBoolBlend {blendOption=Linear}
    alphaScaleBiasClamp {bInterpResult=true}
    boneToModify {boneName=base_hoodtop_m}
    constraintSetup [{transformType=Rotation, targetBone {boneName=dyn_hood_01_m}}]
    constraintWeights [1]
   }
   n5 = Anim Dynamics - Bone: dyn_zipper_m!(ComponentPose=@n4, Alpha=1, GravityOverride:Vector=@n2, ExternalForce="0,0,0") -> (Pose) {
    angularDampingOverride=0.999
    bGravityOverrideInSimSpace=true
    bOverrideAngularDamping=true
    bUseGravityOverride=true
    simulationSpace=BoneRelative
    alphaBoolBlend {blendOption=Linear}
    boundBone {boneName=dyn_zipper_m}
    componentAppliedLinearAccClamp=<3000 3000 3000>
    componentLinearAccScale=<0.3 0.3 0.3>
    componentLinearVelScale=<0.3 0.3 0.3>
    gravityOverride {z=-980}
    physicsBodyDefinitions [{
      boundBone {boneName=dyn_zipper_m}
      boxExtents=<10 10 10>
      constraintSetup {angularLimitsMax=<0 12 0>, angularLimitsMin=<0 -90 0>, angularTarget=<0 0 0>}
      localJointOffset=<0.1 0.1 0.1>
     }]
    relativeSpaceBone {boneName=root}
   }
   n6 = Anim Dynamics - Bone: base_knife_l!(ComponentPose=@n5, Alpha=1, GravityOverride:Vector=@n2, ExternalForce="0,0,0") -> (Pose) {
    angularDampingOverride=0.999
    bGravityOverrideInSimSpace=true
    bOverrideAngularDamping=true
    bUseGravityOverride=true
    simulationSpace=BoneRelative
    alphaBoolBlend {blendOption=Linear}
    boundBone {boneName=base_knife_l}
    componentAppliedLinearAccClamp=<3000 3000 3000>
    componentLinearAccScale=<0.3 0.4 0.3>
    componentLinearVelScale=<0.3 0.4 0.3>
    gravityOverride {z=-980}
    physicsBodyDefinitions [{
      boundBone {boneName=base_knife_l}
      boxExtents=<10 10 10>
      constraintSetup {angularLimitsMax=<0 5 2.5>, angularLimitsMin=<0 -5 -2.5>, angularTarget=<0 0 0>}
      localJointOffset=<0.1 0.1 0.1>
     }]
    relativeSpaceBone {boneName=spine_01}
   }
   n7 = Anim Dynamics - Bone: dyn_shoepull_01_l!(ComponentPose=@n6, Alpha=1, GravityOverride:Vector=@n2, ExternalForce="0,0,0") -> (Pose) {
    angularDampingOverride=0.999
    bGravityOverrideInSimSpace=true
    bUseGravityOverride=true
    simulationSpace=BoneRelative
    alphaBoolBlend {blendOption=Linear}
    boundBone {boneName=dyn_shoepull_01_l}
    componentAppliedLinearAccClamp=<3000 3000 3000>
    componentLinearAccScale=<0.3 0.3 0.3>
    componentLinearVelScale=<0.3 0.3 0.3>
    gravityOverride {z=-980}
    physicsBodyDefinitions [{
      boundBone {boneName=dyn_shoepull_01_l}
      boxExtents=<10 12 12>
      constraintSetup {angularLimitsMax=<0 80 0>, angularLimitsMin=<0 -12.5 0>, angularTarget=<0 0 0>}
      localJointOffset=<0.1 0.1 0.1>
     }]
    relativeSpaceBone {boneName=root}
   }
   n8 = Anim Dynamics - Bone: dyn_shoepull_02_l!(ComponentPose=@n7, Alpha=1, GravityOverride:Vector=@n2, ExternalForce="0,0,0") -> (Pose) {
    angularDampingOverride=0.999
    bGravityOverrideInSimSpace=true
    bUseGravityOverride=true
    simulationSpace=BoneRelative
    alphaBoolBlend {blendOption=Linear}
    boundBone {boneName=dyn_shoepull_02_l}
    componentAppliedLinearAccClamp=<3000 3000 3000>
    componentLinearAccScale=<0.3 0.3 0.3>
    componentLinearVelScale=<0.3 0.3 0.3>
    gravityOverride {z=-980}
    physicsBodyDefinitions [{
      boundBone {boneName=dyn_shoepull_02_l}
      boxExtents=<10 12 12>
      constraintSetup {angularLimitsMax=<0 20 0>, angularLimitsMin=<0 -20 0>, angularTarget=<0 0 0>}
      localJointOffset=<0.1 0.1 0.1>
     }]
    relativeSpaceBone {boneName=root}
   }
   n9 = Anim Dynamics - Bone: dyn_shoepull_01_r!(ComponentPose=@n8, Alpha=1, GravityOverride:Vector=@n2, ExternalForce="0,0,0") -> (Pose) {
    angularDampingOverride=0.999
    bGravityOverrideInSimSpace=true
    bUseGravityOverride=true
    simulationSpace=BoneRelative
    alphaBoolBlend {blendOption=Linear}
    boundBone {boneName=dyn_shoepull_01_r}
    componentAppliedLinearAccClamp=<3000 3000 3000>
    componentLinearAccScale=<0.3 0.3 0.3>
    componentLinearVelScale=<0.3 0.3 0.3>
    gravityOverride {z=-980}
    physicsBodyDefinitions [{
      boundBone {boneName=dyn_shoepull_01_r}
      boxExtents=<10 10 10>
      constraintSetup {angularLimitsMax=<0 80 0>, angularLimitsMin=<0 -12.5 0>, angularTarget=<0 0 0>}
      localJointOffset=<0 0 0>
     }]
    relativeSpaceBone {boneName=root}
   }
   n10 = Anim Dynamics - Bone: dyn_shoepull_01_r(ComponentPose=@n9, Alpha=1, GravityOverride=@n2, ExternalForce="0,0,0") {
    angularDampingOverride=0.999
    bGravityOverrideInSimSpace=true
    bUseGravityOverride=true
    simulationSpace=BoneRelative
    alphaBoolBlend {blendOption=Linear}
    boundBone {boneName=dyn_shoepull_01_r}
    componentAppliedLinearAccClamp=<3000 3000 3000>
    componentLinearAccScale=<0.3 0.3 0.3>
    componentLinearVelScale=<0.3 0.3 0.3>
    gravityOverride {z=-980}
    physicsBodyDefinitions [{
      boundBone {boneName=dyn_shoepull_01_r}
      boxExtents=<10 10 10>
      constraintSetup {angularLimitsMax=<0 20 0>, angularLimitsMin=<0 -20 0>, angularTarget=<0 0 0>}
      localJointOffset=<0 0 0>
     }]
    relativeSpaceBone {boneName=root}
   }
   Save cached pose 'CRCache_Dyn'(Pose=Component To Local(ComponentPose=@n10)) {cacheName=CRCache_Dyn}
   n11 = "Constraint - base_zipper_m (PPPP)"!(
    ComponentPose=Local To Component(LocalPose=Use cached pose 'CRCache_ControlRig'())
    Alpha=1
    ConstraintWeights_0=0.444
    ConstraintWeights_1:float=0.226
    ConstraintWeights_2:float=0.246
    ConstraintWeights_3:float=0.084
   ) -> (Pose) {
    alphaBoolBlend {blendOption=Linear}
    boneToModify {boneName=base_zipper_m}
    constraintSetup [{targetBone {boneName=spine_05}}, {targetBone {boneName=neck_02}}, {targetBone {boneName=head}}, {targetBone {boneName=neck_01}}]
    constraintWeights [1,1,1,1]
   }
   n12 = "Constraint - base_knife_l (PPPPPPPP)"(
    ComponentPose=@n11
    Alpha=1
    ConstraintWeights_0=0.415
    ConstraintWeights_1=0.2617
    ConstraintWeights_2=0.1348
    ConstraintWeights_3=0.0691
    ConstraintWeights_4=0.0513
    ConstraintWeights_5=0.0356
    ConstraintWeights_6=0.0165
    ConstraintWeights_7=0.016
   ) {
    alphaBoolBlend {blendOption=Linear}
    boneToModify {boneName=base_knife_l}
    constraintSetup [
     {targetBone {boneName=spine_05}}
     {targetBone {boneName=clavicle_l}}
     {targetBone {boneName=spine_04}}
     {targetBone {boneName=spine_03}}
     {targetBone {boneName=deform_pec_l}}
     {targetBone {boneName=deform_upperarm_up_l}}
     {targetBone {boneName=neck_02}}
     {targetBone {boneName=spine_02}}
    ]
    constraintWeights [1,1,1,1,1,1,1,1]
   }
   Save cached pose 'CRCache_Const'(Pose=Component To Local(ComponentPose=@n12)) {cacheName=CRCache_Const}
   n13 = Get isSkyDive() -> (isSkyDive:double)
   n14 = "Transform (Modify) Bone - Bone: base_hoodtop_m"(
    ComponentPose=Local To Component(LocalPose=Use cached pose 'CRCache_Const'())
    Alpha=@n13
    Translation="3,0,5.5"
    Rotation="0,0,0"
    Scale="1,1,1"
   ) {
    translationMode=BMM_Additive
    translationSpace=BCS_BoneSpace
    alphaBoolBlend {blendOption=Linear}
    alphaScaleBiasClamp {bInterpResult=true}
    boneToModify {boneName=base_hoodtop_m}
    translation {x=3, z=5.5}
   }
   n15 = "Transform (Modify) Bone - Bone: base_hood_01_m"(ComponentPose=@n14, Alpha=@n13, Translation="3,0,0", Rotation="0,0,0", Scale="1,1,1") {
    translationMode=BMM_Additive
    translationSpace=BCS_BoneSpace
    alphaBoolBlend {blendOption=Linear}
    alphaScaleBiasClamp {bInterpResult=true}
    boneToModify {boneName=base_hood_01_m}
    translation {x=3}
   }
   Save cached pose 'CRCache_Hood'(Pose=Component To Local(ComponentPose=@n15)) {cacheName=CRCache_Hood}
   n16 = Get isJumpAlpha() -> (isJumpAlpha:double)
   n17 = "Transform (Modify) Bone - Bone: base_sidepouch_r"(
    ComponentPose=Local To Component(LocalPose=Use cached pose 'CRCache_Hood'())
    Alpha=@n16
    Translation="0,0,0"
    Rotation="0,0,-15"
    Scale="1,1,1"
   ) {
    rotationMode=BMM_Additive
    rotationSpace=BCS_BoneSpace
    translationMode=BMM_Additive
    translationSpace=BCS_BoneSpace
    alphaBoolBlend {blendOption=Linear}
    alphaScaleBiasClamp {bInterpResult=true}
    boneToModify {boneName=base_sidepouch_r}
    rotation {roll=-15}
   }
   n18 = "Transform (Modify) Bone - Bone: base_sidepouch_l"(ComponentPose=@n17, Alpha=@n16, Translation="0,0,0", Rotation="-2,0,0", Scale="1,1,1") {
    rotationMode=BMM_Additive
    rotationSpace=BCS_BoneSpace
    translationMode=BMM_Additive
    translationSpace=BCS_BoneSpace
    alphaBoolBlend {blendOption=Linear}
    alphaScaleBiasClamp {bInterpResult=true}
    boneToModify {boneName=base_sidepouch_l}
    rotation {pitch=-2}
   }
   n19 = "Transform (Modify) Bone - Bone: base_sidepouch_l"(ComponentPose=@n18, Alpha=Get isCrouchedAlpha() -> (isCrouchedAlpha:double), Translation="0,0,0", Rotation="-5,0,0", Scale="1,1,1") {
    rotationMode=BMM_Additive
    rotationSpace=BCS_BoneSpace
    translationMode=BMM_Additive
    translationSpace=BCS_BoneSpace
    alphaBoolBlend {blendOption=Linear}
    alphaScaleBiasClamp {bInterpResult=true}
    boneToModify {boneName=base_sidepouch_l}
    rotation {pitch=-5}
   }
   n20 = "Transform (Modify) Bone - Bone: base_belt_l"(ComponentPose=@n19, Alpha=Get isCrouchedAlpha(), Translation="-3.345,-0.097,-10.022", Rotation="-45,0,0", Scale="1,1,1") {
    rotationMode=BMM_Additive
    rotationSpace=BCS_BoneSpace
    translationMode=BMM_Additive
    translationSpace=BCS_BoneSpace
    alphaBoolBlend {blendOption=Linear}
    alphaScaleBiasClamp {bInterpResult=true}
    boneToModify {boneName=base_belt_l}
    rotation {pitch=-45}
    translation=<-3.345 -0.097 -10.022>
   }
   Save cached pose 'CRCache_Legs'(Pose=Component To Local(ComponentPose=@n20)) {cacheName=CRCache_Legs}
   n21 = "Transform (Modify) Bone - Bone: base_zipper_m"(
    ComponentPose=Local To Component(LocalPose=Use cached pose 'CRCache_Dyn'())
    Alpha=Get isLookDown() -> (isLookDown:double)
    Translation="0,0,0"
    Rotation="26,0,0"
    Scale="1,1,1"
   ) {
    rotationMode=BMM_Additive
    rotationSpace=BCS_BoneSpace
    translationMode=BMM_Additive
    translationSpace=BCS_BoneSpace
    alphaBoolBlend {blendOption=Linear}
    alphaScaleBiasClamp {bInterpResult=true, interpSpeedDecreasing=5, interpSpeedIncreasing=4}
    boneToModify {boneName=base_zipper_m}
    rotation {pitch=26}
   }
   Save cached pose 'CRCache_Zipper'(Pose=Component To Local(ComponentPose=@n21)) {cacheName=CRCache_Zipper}
  ]}
 EventGraph {nodes [
   Event BlueprintUpdateAnimation()
   > n1 = Sequence!(execute) -> (then_0, then_1, then_2:Exec, then_3:Exec, then_4:Exec)
   n2 = NOT Boolean(A=Get bIsParachuteOpened())
   n3 = OR Boolean!(A=Get bIsSkydiveDiveMode(), B=Get bIsSkyDiving(), C:bool=Get bIsSkydiveFloating()) -> (ReturnValue)
   Set isSkyDive(execute=@n1.then_3, isSkyDive:double="To Float (Boolean)"(InBool=AND Boolean(A=@n3, B=@n2))) -> (Output_Get:double)
   n4 = Get bIsFalling()
   n5 = Get bIsJumping()
   > Set isJumpAlpha(isJumpAlpha:double="To Float (Boolean)"(InBool=OR Boolean!(A=@n5, B=@n4) -> (ReturnValue))) -> (Output_Get:double)
   n6 = Get bIsCrouchSprinting()
   n7 = Get bIsCrouching()
   n8 = Get bVehicleSeatForcesCrouch() -> (bVehicleSeatForcesCrouch:bool)
   > Set isCrouchedAlpha(isCrouchedAlpha:double="To Float (Boolean)"(InBool=OR Boolean!(A=@n7, B=@n6, C:bool=@n8) -> (ReturnValue))) -> (Output_Get:double)
   n9 = Branch(
    execute=@n1.then_2
    Condition=OR Boolean!(A=Get bIsSkydiveDiveMode(), B=Get bIsSkyDiving(), C=Get bIsSkydiveFloating(), D:bool=Get bIsParachuteOpened()) -> (ReturnValue)
   )
   > Set AlphaConstraintHood(AlphaConstraintHood:double=0.65) -> (Output_Get:double)
   Set AlphaConstraintHood(execute=@n9.else, AlphaConstraintHood=0.45)
   n10 = Gravity Override State Function DEPRECATED(
    execute=@n1.then_0
    CharacterPartAnimInstance=Self-Reference()
    GameTimeInSeconds=GetGameTimeInSeconds()
    GravityOverrideParams=Get Gravity Override Params()
   )
   > Set GravityOverride(GravityOverride=@n10)
   > n11 = Branch(Condition=OR Boolean(A=Get bIsSkyDiving(), B=Get bIsSkydiveDiveMode(), C=Get bIsParachuteOpened(), D=Get bIsSkydiveFloating()))
   > Set Sim Space Settings SkyDive(Sim Space Settings SkyDive:Sim Space Settings Structure=Get Sim Space Settings Skydiving() -> (Sim Space Settings Skydiving:Sim Space Settings Structure)) -> (Output_Get:Sim Space Settings Structure)
   Set Sim Space Settings SkyDive(execute=@n11.else, Sim Space Settings SkyDive=Get Sim Space Settings Regular() -> (Sim Space Settings Regular:Sim Space Settings Structure))
   n12 = BreakTransform(InTransform=GetDeltaTransformFromRefPose(self=GetOwnerSkeletalMeshComponent(), BoneName=neck_01, BaseName=spine_05))
   n13 = Branch(execute=@n1.then_4, Condition="float >= float"(A=BreakRotator(InRot=@n12.Rotation).Yaw, B=-10))
   > Set isLookDown(isLookDown:double=1) -> (Output_Get:double)
   Set isLookDown(execute=@n13.else, isLookDown=0)
   n14 = Branch(
    execute=@n1.then_1
    Condition=OR Boolean(A=Get bIsSkydiveDiveMode(), B=Get bIsSkyDiving(), C=Get bIsSkydiveFloating(), D=Get bIsParachuteOpened())
   )
   > Set AlphaDynHood(AlphaDynHood:double=0.65) -> (Output_Get:double)
   Set AlphaDynHood(execute=@n14.else, AlphaDynHood=1)
  ]}
]
variables [
 AlphaConstraintHood {type=number}
 AlphaDynHood {type=number}
 Gravity Override Params {type=GravityOverrideParamsStruct}
 GravityOverride {type=Vector}
 isCrouchedAlpha {type=number}
 isJumpAlpha {type=number}
 isLookDown {type=number}
 isSkyDive {type=number}
 Sim Space Settings Regular {
  type=SimSpaceSettings
  default {
   dampingAlpha=1
   maxAngularAcceleration=1000
   maxAngularVelocity=3
   maxLinearAcceleration=700
   maxLinearVelocity=500
   velocityScaleZ=0
   worldAlpha=0.3
   externalLinearDragV=<0.2 0.2 0>
  }
 }
 Sim Space Settings SkyDive {
  type=SimSpaceSettings
  default {
   dampingAlpha=1
   maxAngularAcceleration=10000
   maxAngularVelocity=10000
   maxLinearAcceleration=10000
   maxLinearVelocity=10000
   velocityScaleZ=1
   worldAlpha=0
   externalLinearDragV=<0 0 0>
  }
 }
 Sim Space Settings Skydiving {
  type=SimSpaceSettings
  default {
   dampingAlpha=1
   maxAngularAcceleration=1000
   maxAngularVelocity=3
   maxLinearAcceleration=700
   maxLinearVelocity=500
   velocityScaleZ=0
   worldAlpha=0.65
   externalLinearDragV=<0.2 0.2 0>
  }
 }
]
