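# Physics asset for the face accessory part of the male medium DriftSwat character, simulating its dynamic hair bones.
# Each bone carries a small capsule body, with base_hair_01_m as the shared root.
# Left and right strands form chains one to three bones long, each bone constrained to its parent back to that root.
# Every joint allows only a slight swing and sets twist limits to zero, keeping the hair stiff.
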PhysicsAsset /BRCosmetics/Characters/Player/Male/Medium/Bodies/M_MED_DriftSwat/Meshes/Parts/M_MED_DriftSwat_FaceAcc_Physics
bodies {
 base_hair_01_m [{shapeType=Capsule, length=0.505, radius=0.505, shapeName=base_hair_01_m_capsule, rotation=<rot 0 -90 0>}]
 dyn_hair_a01_l [{shapeType=Capsule, length=1.5, radius=0.7, shapeName=dyn_hair_a01_l_capsule, center=<1.527 0 0>, rotation=<rot 90 0 0>}]
 dyn_hair_a01_r [{shapeType=Capsule, length=1.5, radius=0.7, shapeName=dyn_hair_a01_r_capsule, center=<1 0 0>, rotation=<rot 90 0 0>}]
 dyn_hair_a02_l [{shapeType=Capsule, length=1.5, radius=0.7, shapeName=dyn_hair_a02_l_capsule, center=<1 0 0>, rotation=<rot 90 0 0>}]
 dyn_hair_a02_r [{shapeType=Capsule, length=1.5, radius=0.7, shapeName=dyn_hair_a02_r_capsule, center=<1.5002 0 0>, rotation=<rot 90 0 0>}]
 dyn_hair_a03_l [{shapeType=Capsule, length=1.5, radius=0.7, shapeName=dyn_hair_a03_l_capsule, center=<1.3298 0 0>, rotation=<rot 90 0 0>}]
 dyn_hair_a03_r [{shapeType=Capsule, length=1.5, radius=0.7, shapeName=dyn_hair_a03_r_capsule, center=<1.3578 0 0>, rotation=<rot 90 0 0>}]
 dyn_hair_a04_l [{shapeType=Capsule, length=1.5, radius=0.7, shapeName=dyn_hair_a04_l_capsule, center=<1 0 0>, rotation=<rot 90 0 0>}]
 dyn_hair_a04_r [{shapeType=Capsule, length=1, radius=0.7, shapeName=dyn_hair_a04_r_capsule, center=<1.1715 0 0>, rotation=<rot 90 0 0>}]
 dyn_hair_a05_l [{shapeType=Capsule, length=1, radius=0.7, shapeName=dyn_hair_a05_l_capsule, center=<1.3712 0 0>, rotation=<rot 90 0 0>}]
 dyn_hair_a05_r [{shapeType=Capsule, length=1.5, radius=0.7, shapeName=dyn_hair_a05_r_capsule, center=<1.3734 0 0>, rotation=<rot 90 0 0>}]
 dyn_hair_a06_l [{shapeType=Capsule, length=1, radius=0.7, shapeName=dyn_hair_a06_l_capsule, center=<1.2566 0 0>, rotation=<rot 90 0 0>}]
 dyn_hair_a06_r [{shapeType=Capsule, length=1, radius=0.7, shapeName=dyn_hair_a06_r_capsule, center=<1.2322 0 0>, rotation=<rot 90 0 0>}]
 dyn_hair_b01_l [{shapeType=Capsule, length=1.5, radius=0.7, shapeName=dyn_hair_b01_l_capsule, center=<1.2591 0 0>, rotation=<rot 90 0 0>}]
 dyn_hair_b02_r [{shapeType=Capsule, length=1.5, radius=0.7, shapeName=dyn_hair_b02_r_capsule, center=<1 0 0>, rotation=<rot 90 0 0>}]
 dyn_hair_b03_l [{shapeType=Capsule, length=1.5, radius=0.7, shapeName=dyn_hair_b03_l_capsule, center=<1.3847 0 0>, rotation=<rot 90 0 0>}]
 dyn_hair_b03_r [{shapeType=Capsule, length=1.5, radius=0.7, shapeName=dyn_hair_b03_r_capsule, center=<1.3217 0 0>, rotation=<rot 90 0 0>}]
 dyn_hair_b04_r [{shapeType=Capsule, length=0.5, radius=0.7, shapeName=dyn_hair_b04_r_capsule, center=<1.0458 0 0>, rotation=<rot 90 0 0>}]
 dyn_hair_b05_l [{shapeType=Capsule, length=0.7, radius=0.7, shapeName=dyn_hair_b05_l_capsule, center=<1.1631 0 0>, rotation=<rot 90 0 0>}]
 dyn_hair_b05_r [{shapeType=Capsule, length=1.5, radius=0.7, shapeName=dyn_hair_b05_r_capsule, center=<1.1417 0 0>, rotation=<rot 90 0 0>}]
 dyn_hair_b06_l [{shapeType=Capsule, length=0.5, radius=0.7, shapeName=dyn_hair_b06_l_capsule, center=<1.115 0 0>, rotation=<rot 90 0 0>}]
 dyn_hair_b06_r [{shapeType=Capsule, length=1, radius=0.7, shapeName=dyn_hair_b06_r_capsule, center=<1.2322 0 0>, rotation=<rot 90 0 0>}]
 dyn_hair_c01_l [{shapeType=Capsule, length=1.5, radius=0.7, shapeName=dyn_hair_c01_l_capsule, center=<1.354 0 0>, rotation=<rot 90 0 0>}]
 dyn_hair_c02_r [{shapeType=Capsule, length=1.5, radius=0.7, shapeName=dyn_hair_c02_r_capsule, center=<1 0 0>, rotation=<rot 90 0 0>}]
 dyn_hair_c03_l [{shapeType=Capsule, length=1.5, radius=0.7, shapeName=dyn_hair_c03_l_capsule, center=<1.3458 0 0>, rotation=<rot 90 0 0>}]
 dyn_hair_c03_r [{shapeType=Capsule, length=1.5, radius=0.7, shapeName=dyn_hair_c03_r_capsule, center=<1.343 0 0>, rotation=<rot 90 0 0>}]
 dyn_hair_c04_r [{shapeType=Capsule, length=0.5, radius=0.7, shapeName=dyn_hair_c04_r_capsule, center=<1.0922 0 0>, rotation=<rot 90 0 0>}]
 dyn_hair_c05_l [{shapeType=Capsule, length=1, radius=0.7, shapeName=dyn_hair_c05_l_capsule, center=<1.1187 0 0>, rotation=<rot 90 0 0>}]
 dyn_hair_c06_l [{shapeType=Capsule, length=1, radius=0.7, shapeName=dyn_hair_c06_l_capsule, center=<1.2244 0 0>, rotation=<rot 90 0 0>}]
}
constraints [
 {bone1Name=dyn_hair_a01_l, bone2Name=base_hair_01_m, swing1LimitDegrees=2.5, swing2LimitDegrees=2.5, twistLimitDegrees=0}
 {bone1Name=dyn_hair_b01_l, bone2Name=dyn_hair_a01_l, swing1LimitDegrees=2.5, swing2LimitDegrees=2.5, twistLimitDegrees=0}
 {bone1Name=dyn_hair_c01_l, bone2Name=dyn_hair_b01_l, swing1LimitDegrees=2.5, swing2LimitDegrees=2.5, twistLimitDegrees=0}
 {bone1Name=dyn_hair_a01_r, bone2Name=base_hair_01_m, swing1LimitDegrees=2.5, swing2LimitDegrees=2.5, twistLimitDegrees=0}
 {bone1Name=dyn_hair_a02_l, bone2Name=base_hair_01_m, swing1LimitDegrees=2.5, swing2LimitDegrees=2.5, twistLimitDegrees=0}
 {bone1Name=dyn_hair_a02_r, bone2Name=base_hair_01_m, swing1LimitDegrees=2.5, swing2LimitDegrees=2.5, twistLimitDegrees=0}
 {bone1Name=dyn_hair_b02_r, bone2Name=dyn_hair_a02_r, swing1LimitDegrees=2.5, swing2LimitDegrees=2.5, twistLimitDegrees=0}
 {bone1Name=dyn_hair_c02_r, bone2Name=dyn_hair_b02_r, swing1LimitDegrees=2.5, swing2LimitDegrees=2.5, twistLimitDegrees=0}
 {bone1Name=dyn_hair_a03_l, bone2Name=base_hair_01_m, swing1LimitDegrees=2.5, swing2LimitDegrees=2.5, twistLimitDegrees=0}
 {bone1Name=dyn_hair_b03_l, bone2Name=dyn_hair_a03_l, swing1LimitDegrees=2.5, swing2LimitDegrees=2.5, twistLimitDegrees=0}
 {bone1Name=dyn_hair_c03_l, bone2Name=dyn_hair_b03_l, swing1LimitDegrees=2.5, swing2LimitDegrees=2.5, twistLimitDegrees=0}
 {bone1Name=dyn_hair_a03_r, bone2Name=base_hair_01_m, swing1LimitDegrees=2.5, swing2LimitDegrees=2.5, twistLimitDegrees=0}
 {bone1Name=dyn_hair_b03_r, bone2Name=dyn_hair_a03_r, swing1LimitDegrees=2.5, swing2LimitDegrees=2.5, twistLimitDegrees=0}
 {bone1Name=dyn_hair_c03_r, bone2Name=dyn_hair_b03_r, swing1LimitDegrees=2.5, swing2LimitDegrees=2.5, twistLimitDegrees=0}
 {bone1Name=dyn_hair_a04_l, bone2Name=base_hair_01_m, swing1LimitDegrees=2.5, swing2LimitDegrees=2.5, twistLimitDegrees=0}
 {bone1Name=dyn_hair_a04_r, bone2Name=base_hair_01_m, swing1LimitDegrees=2.5, swing2LimitDegrees=2.5, twistLimitDegrees=0}
 {bone1Name=dyn_hair_b04_r, bone2Name=dyn_hair_a04_r, swing1LimitDegrees=2.5, swing2LimitDegrees=2.5, twistLimitDegrees=0}
 {bone1Name=dyn_hair_c04_r, bone2Name=dyn_hair_b04_r, swing1LimitDegrees=2.5, swing2LimitDegrees=2.5, twistLimitDegrees=0}
 {bone1Name=dyn_hair_a05_l, bone2Name=base_hair_01_m, swing1LimitDegrees=2.5, swing2LimitDegrees=2.5, twistLimitDegrees=0}
 {bone1Name=dyn_hair_b05_l, bone2Name=dyn_hair_a05_l, swing1LimitDegrees=2.5, swing2LimitDegrees=2.5, twistLimitDegrees=0}
 {bone1Name=dyn_hair_c05_l, bone2Name=dyn_hair_b05_l, swing1LimitDegrees=2.5, swing2LimitDegrees=2.5, twistLimitDegrees=0}
 {bone1Name=dyn_hair_a05_r, bone2Name=base_hair_01_m, swing1LimitDegrees=2.5, swing2LimitDegrees=2.5, twistLimitDegrees=0}
 {bone1Name=dyn_hair_b05_r, bone2Name=dyn_hair_a05_r, swing1LimitDegrees=2.5, swing2LimitDegrees=2.5, twistLimitDegrees=0}
 {bone1Name=dyn_hair_a06_l, bone2Name=base_hair_01_m, swing1LimitDegrees=2.5, swing2LimitDegrees=2.5, twistLimitDegrees=0}
 {bone1Name=dyn_hair_b06_l, bone2Name=dyn_hair_a06_l, swing1LimitDegrees=2.5, swing2LimitDegrees=2.5, twistLimitDegrees=0}
 {bone1Name=dyn_hair_c06_l, bone2Name=dyn_hair_b06_l, swing1LimitDegrees=2.5, swing2LimitDegrees=2.5, twistLimitDegrees=0}
 {bone1Name=dyn_hair_a06_r, bone2Name=base_hair_01_m, swing1LimitDegrees=2.5, swing2LimitDegrees=2.5, twistLimitDegrees=0}
 {bone1Name=dyn_hair_b06_r, bone2Name=dyn_hair_a06_r, swing1LimitDegrees=2.5, swing2LimitDegrees=2.5, twistLimitDegrees=0}
]
